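; Camera modifier base for in-air states in GrillMoon that shifts the view along the camera's up axis.
; BlueprintModifyCamera samples a float curve with the modifier's Alpha and uses the result to lerp from zero to UpOffset.
; That amount scales the view rotation's up vector and is added to the location, leaving rotation and FOV unchanged.
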(fn BlueprintModifyCamera
  (params (DeltaTime :type Float%20%28single-precision%29) (ViewLocation :type Vector :default "0, 0, 0") (ViewRotation :type Rotator :default "0, 0, 0") (FOV :type Float%20%28single-precision%29))
  (return (+ ViewLocation (* (Math|Vector|GetUpVector ViewRotation) (Math|Float|Lerp 0.0 (Variables|Default|GetUpOffset) (Math|Curves|GetFloatValue (Variables|Default|GetCurve) (Variables|CameraModifier|GetAlpha))))) ViewRotation FOV))
